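# Sound cue for the final impact when a demolition Chariot object is destroyed.
# A Random node picks one of two equally weighted Chariot final-impact waves.
# A Modulator then varies pitch slightly downward for variation.
# It plays at reduced volume in the impacts sound class with small-object destruction attenuation and a four-voice object-impact concurrency limit.
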SoundCue /Game/Sounds/Fort_Impact_Sounds/Demolition/Chariot/Demo_Chariot_ImpactFinal_Cue
root=Modulator_2
nodes [
 {id=Modulator_2, children [Random_2], properties {pitchMax=1, pitchMin=0.92}}
 {id=Random_2, children [WavePlayer_3,WavePlayer_7], properties {weights [1,1]}}
 {id=WavePlayer_3, properties {soundWaveAssetPtr=&/Chariot_FinalImpact_01}}
 {id=WavePlayer_7, properties {soundWaveAssetPtr=&/Chariot_FinalImpact_02}}
]
properties {
 attenuationSettings=/Game/Sounds/Attenuation/Fort_Demolition_Destroy_Small_Sized_Object_Att
 maxDistance=4000
 soundClassObject=&Fort_Impacts_PSM
 volumeMultiplier=0.6
 concurrencyOverrides {maxCount=4}
 concurrencySet [&CONC_Object_Impacts_FourMax]
}
