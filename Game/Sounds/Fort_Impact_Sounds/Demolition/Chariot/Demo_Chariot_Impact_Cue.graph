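# Sound cue for the demolition impact sound of a chariot object.
# A random node picks one of two equally weighted impact waves, then a modulator varies its pitch slightly downward.
# The cue uses small-object demolition hit attenuation, the impacts sound class and a concurrency set that caps simultaneous object impacts at four.
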SoundCue /Game/Sounds/Fort_Impact_Sounds/Demolition/Chariot/Demo_Chariot_Impact_Cue
root=Modulator_2
nodes [
 {id=Modulator_2, children [Random_1], properties {pitchMax=1, pitchMin=0.92}}
 {id=Random_1, children [WavePlayer_3,WavePlayer_4], properties {weights [1,1]}}
 {id=WavePlayer_3, properties {soundWaveAssetPtr=&/Chariot_Impact_01}}
 {id=WavePlayer_4, properties {soundWaveAssetPtr=&/Chariot_Impact_02}}
]
properties {
 attenuationSettings=/Game/Sounds/Attenuation/Fort_Demolition_Hit_Small_Sized_Object_Att
 maxDistance=4000
 soundClassObject=&Fort_Impacts
 volumeMultiplier=0.6
 concurrencyOverrides {maxCount=4}
 concurrencySet [&CONC_Object_Impacts_FourMax]
}
